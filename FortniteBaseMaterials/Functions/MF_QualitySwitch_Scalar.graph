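# Scalar material function that selects one of five inputs (Epic, High, Low, Medium, Mobile), each annotated with its target platforms, by shading path and quality level.
# On default and deferred paths, a quality switch picks Epic, High or Low, with unset levels falling back to Epic.
# On the mobile path, a second switch uses Mobile at low quality and Medium otherwise.
# An unconnected dashed input visually separates the two groups.
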MaterialFunction /FortniteBaseMaterials/Functions/MF_QualitySwitch_Scalar
graph {nodes [
  FunctionInput() {bUsePreviewValueAsDefault=true, inputName=--------, inputType=FunctionInput_Scalar, sortPriority=3}
  n1 = FunctionInput() {
   bUsePreviewValueAsDefault=true
   desc="5th Gen Consoles, PC Epic, PC High"
   description="5th Gen Consoles, PC Epic, PC High"
   inputName=Epic
   inputType=FunctionInput_Scalar
   sortPriority=2
  }
  n2 = FunctionInput() {
   bUsePreviewValueAsDefault=true
   desc="PC Low, PC Medium (not Perf mode)"
   description="PC Low, PC Medium (not Perf mode)"
   inputName=Low
   inputType=FunctionInput_Scalar
  }
  n3 = FunctionInput() {
   bUsePreviewValueAsDefault=true
   desc="4th Gen Consoles\r\nno SSColor"
   description=4th Gen Consoles
   inputName=High
   inputType=FunctionInput_Scalar
   sortPriority=1
  }
  n4 = QualitySwitch(Default=@n1, Low=@n2, High=@n3, Epic=@n1) {desc=has GBuffers}
  n5 = FunctionInput() {
   bUsePreviewValueAsDefault=true
   desc="Portable Consoles\r\nhas outlines?\r\nno roughness outlines"
   description=Portable Consoles
   inputName=Medium
   inputType=FunctionInput_Scalar
   sortPriority=5
  }
  n6 = FunctionInput() {
   bUsePreviewValueAsDefault=true
   desc="Android, PC Perf\r\nno outlines (depth buffer cant be transformed)\r\nno roughness outlines"
   description="Android, PC Perf"
   inputName=Mobile
   inputType=FunctionInput_Scalar
   sortPriority=4
  }
  FunctionOutput(None=ShadingPathSwitch(Default=@n4, Deferred=@n4, Mobile=QualitySwitch(Default=@n5, Low=@n6, Medium=@n5) {desc="no GBuffers, has depth"}))
 ]}
